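# MAT_HydeT1_Decal is a masked, default-lit material for decals on the HydeT1 vehicle body cosmetic.
# Its graph holds a near-black constant color and a sample of T_HydeT1_Decals_N read as linear color, with no connections shown to material outputs.
# Usage is enabled for skeletal meshes and mesh deformers, with automatic usage setting in the editor turned off.
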Material /VehicleCosmetics/Bodies/HydeT1/Materials/MAT_HydeT1_Decal
graph {
 properties {
  bAutomaticallySetUsageInEditor=false
  bUsedWithMeshDeformer=true
  bUsedWithSkeletalMesh=true
  blendMode=BLEND_Masked
  usedShadingModels=Default Lit
 }
 nodes [
  Constant3Vector() {constant=<rgba 0.02 0.02 0.02 0>}
  TextureSample() {samplerType=SAMPLERTYPE_LinearColor, texture=/VehicleCosmetics/Bodies/HydeT1/Textures/T_HydeT1_Decals_N}
 ]
}
